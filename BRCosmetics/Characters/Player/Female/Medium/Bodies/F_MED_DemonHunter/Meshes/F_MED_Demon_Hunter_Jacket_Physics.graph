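# Physics asset for the jacket on the medium female Demon Hunter body cosmetic, giving its back pieces secondary motion.
# Two box bodies on the dyn_jacket_back and dyn_jacket_back_r bones represent the rear jacket flaps, alongside a capsule on spine_02 for the torso.
# Each flap is constrained directly to spine_02 with identical limited swing and tight twist ranges, keeping its movement close to the back.
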PhysicsAsset /BRCosmetics/Characters/Player/Female/Medium/Bodies/F_MED_DemonHunter/Meshes/F_MED_Demon_Hunter_Jacket_Physics
bodies {
 dyn_jacket_back [{shapeType=Box, extentX=2.5419, extentY=7.8876, extentZ=10.9734, center=<2.3289 -0.0017 0.0673>, rotation=<rot -1.7692 -90 -1.6566>}]
 dyn_jacket_back_r [{shapeType=Box, extentX=2.5419, extentY=7.8876, extentZ=10.9109, center=<2.3289 -0.0017 0.0673>, rotation=<rot -1.7692 -90 -1.6566>}]
 spine_02 [{shapeType=Capsule, length=6.5, radius=10.75, center=<-3.5662 2.1623 -0.0002>, rotation=<rot 90 0 -10.0002>}]
}
constraints [
 {bone1Name=dyn_jacket_back, bone2Name=spine_02, swing1LimitDegrees=23, swing2LimitDegrees=13, twistLimitDegrees=5}
 {bone1Name=dyn_jacket_back_r, bone2Name=spine_02, swing1LimitDegrees=23, swing2LimitDegrees=13, twistLimitDegrees=5}
]
