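# Post-copy AnimBlueprint for the female Demon Hunter cosmetic body, adding secondary motion to jacket, sleeves and hood.
# It copies the attached parent's pose, offsets jacket bones using head yaw and skydive state, then simulates sleeves, RigidBody physics and hood.
# That cached pose feeds two jacket simulations blended by crouch state, with extra RigidBody physics when crouched at low speed.
# Jumping triggers a delayed upward external force on hood and jacket.
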AnimBlueprint /BRCosmetics/Characters/Player/Female/Medium/Bodies/F_MED_DemonHunter/Meshes/F_MED_Demon_Hunter_AnimBp
target_skeleton=&/F_MED_Demon_Hunter_Skeleton
graphs [
 AnimGraph {nodes [
   n1 = Get ext_force() -> (ext_force:Vector)
   n2 = Anim Dynamics - Bone: dyn_jacket_r(ComponentPose=Local To Component(LocalPose=Use cached pose 'CrouchingCache'()), Alpha=0.6, ExternalForce=@n1) {
    angularDampingOverride=0.9999
    bGravityOverrideInSimSpace=true
    bOverrideAngularDamping=true
    bOverrideLinearDamping=true
    gravityScale=0.75
    linearDampingOverride=0.9999
    simulationSpace=RootRelative
    alphaBoolBlend {blendOption=Linear}
    boundBone {boneName=dyn_jacket_r}
    componentAppliedLinearAccClamp=<1000 1000 1000>
    componentLinearAccScale=<0.5 0.5 0.5>
    componentLinearVelScale=<0.5 0.5 0.5>
    physicsBodyDefinitions [{
      sphereCollisionRadius=0
      boundBone {boneName=dyn_jacket_r}
      boxExtents=<9 3 16>
      constraintSetup {angularLimitsMax=<0 30 10>, angularLimitsMin=<0 -30 -10>, angularTarget=<0 0 0>}
      localJointOffset=<2 0 0>
     }]
    relativeSpaceBone {boneName=pelvis}
   }
   n3 = Anim Dynamics - Bone: dyn_jacket_l(ComponentPose=@n2, Alpha=0.6, ExternalForce=@n1) {
    angularDampingOverride=0.9999
    bGravityOverrideInSimSpace=true
    bOverrideAngularDamping=true
    bOverrideLinearDamping=true
    gravityScale=0.75
    linearDampingOverride=0.9999
    simulationSpace=RootRelative
    alphaBoolBlend {blendOption=Linear}
    boundBone {boneName=dyn_jacket_l}
    componentAppliedLinearAccClamp=<1000 1000 1000>
    componentLinearAccScale=<0.5 0.5 0.5>
    componentLinearVelScale=<0.5 0.5 0.5>
    physicsBodyDefinitions [{
      sphereCollisionRadius=0
      boundBone {boneName=dyn_jacket_l}
      boxExtents=<9 16 3>
      constraintSetup {angularLimitsMax=<0 50 5>, angularLimitsMin=<0 -50 -5>, angularTarget=<0 0 0>}
      localJointOffset=<2 0 0>
     }]
    relativeSpaceBone {boneName=pelvis}
   }
   n4 = RigidBody(ComponentPose=@n3, Alpha=Get Jacket_Crouch() -> (Jacket_Crouch:double), ExternalForce=@n1) {
    bClampLinearTranslationLimitToRefPose=true
    bOverrideWorldGravity=true
    overridePhysicsAsset=&/F_MED_Demon_Hunter_Jacket_Physics
    simulationSpace=BaseBoneSpace
    alphaBoolBlend {blendOption=Linear}
    baseBoneRef {boneName=root}
    componentAppliedLinearAccClamp=<1000 1000 1000>
    componentLinearAccScale=<1 1 1>
    componentLinearVelScale=<1 1 1>
    overrideWorldGravity {z=-980}
   }
   n5 = Get ext_force()
   n6 = Anim Dynamics - Bone: dyn_jacket_back_r(ComponentPose=Local To Component(LocalPose=Use cached pose 'CrouchingCache'()), Alpha=0.2, ExternalForce=@n5) {
    angularDampingOverride=1
    bGravityOverrideInSimSpace=true
    bOverrideAngularDamping=true
    bOverrideLinearDamping=true
    linearDampingOverride=1
    simulationSpace=RootRelative
    alphaBoolBlend {blendOption=Linear}
    boundBone {boneName=dyn_jacket_back_r}
    componentAppliedLinearAccClamp=<500 500 500>
    componentLinearAccScale=<0.5 0.5 0.25>
    componentLinearVelScale=<0.5 0.5 0.25>
    physicsBodyDefinitions [{
      sphereCollisionRadius=0
      boundBone {boneName=dyn_jacket_back_r}
      boxExtents=<8 2 9>
      constraintSetup {angularLimitsMax=<0 20 1>, angularLimitsMin=<0 -20 -20>, angularTarget=<0 0 0>}
      localJointOffset=<3 0 0>
     }]
    relativeSpaceBone {boneName=pelvis}
   }
   n7 = Anim Dynamics - Bone: dyn_jacket_back(ComponentPose=@n6, Alpha=0.3, ExternalForce=@n5) {
    angularDampingOverride=1
    bGravityOverrideInSimSpace=true
    bOverrideAngularDamping=true
    bOverrideLinearDamping=true
    gravityScale=0.75
    linearDampingOverride=1
    simulationSpace=RootRelative
    alphaBoolBlend {blendOption=Linear}
    boundBone {boneName=dyn_jacket_back}
    componentAppliedLinearAccClamp=<400 400 400>
    componentLinearAccScale=<0.25 0.25 0.5>
    componentLinearVelScale=<0.25 0.25 0.5>
    physicsBodyDefinitions [{
      sphereCollisionRadius=0
      boundBone {boneName=dyn_jacket_back}
      boxExtents=<8 2 10>
      constraintSetup {angularLimitsMax=<0 15 -5>, angularLimitsMin=<0 -15 -25>, angularTarget=<0 0 0>}
      localJointOffset=<3 0 0>
     }]
    relativeSpaceBone {boneName=root}
   }
   n8 = Anim Dynamics - Bone: dyn_jacket_r(ComponentPose=@n7, Alpha=0.6, ExternalForce=@n5) {
    angularDampingOverride=0.9999
    bGravityOverrideInSimSpace=true
    bOverrideAngularDamping=true
    bOverrideLinearDamping=true
    gravityScale=0.75
    linearDampingOverride=0.9999
    simulationSpace=RootRelative
    alphaBoolBlend {blendOption=Linear}
    boundBone {boneName=dyn_jacket_r}
    componentAppliedLinearAccClamp=<1000 1000 1000>
    componentLinearAccScale=<0.5 0.5 0.5>
    componentLinearVelScale=<0.5 0.5 0.5>
    physicsBodyDefinitions [{
      sphereCollisionRadius=0
      boundBone {boneName=dyn_jacket_r}
      boxExtents=<9 3 16>
      constraintSetup {angularLimitsMax=<0 30 10>, angularLimitsMin=<0 -30 -10>, angularTarget=<0 0 0>}
      localJointOffset=<2 0 0>
     }]
    relativeSpaceBone {boneName=pelvis}
   }
   n9 = Anim Dynamics - Bone: dyn_jacket_l(ComponentPose=@n8, Alpha=0.6, ExternalForce=@n5) {
    angularDampingOverride=0.9999
    bGravityOverrideInSimSpace=true
    bOverrideAngularDamping=true
    bOverrideLinearDamping=true
    gravityScale=0.75
    linearDampingOverride=0.9999
    simulationSpace=RootRelative
    alphaBoolBlend {blendOption=Linear}
    boundBone {boneName=dyn_jacket_l}
    componentAppliedLinearAccClamp=<1000 1000 1000>
    componentLinearAccScale=<0.5 0.5 0.5>
    componentLinearVelScale=<0.5 0.5 0.5>
    physicsBodyDefinitions [{
      sphereCollisionRadius=0
      boundBone {boneName=dyn_jacket_l}
      boxExtents=<9 16 3>
      constraintSetup {angularLimitsMax=<0 50 5>, angularLimitsMin=<0 -50 -5>, angularTarget=<0 0 0>}
      localJointOffset=<2 0 0>
     }]
    relativeSpaceBone {boneName=pelvis}
   }
   Output Pose(Result=Blend Poses by bool(BlendPose_0=Component To Local(ComponentPose=@n4), BlendPose_1=Component To Local(ComponentPose=@n9), bActiveValue=Get bIsCrouching()) {blendType=Linear}) {layerGroup=None}
   n10 = Get Jacket_Rotate() -> (Jacket_Rotate:double)
   n11 = "Transform (Modify) Bone - Bone: dyn_jacket_back"(
    ComponentPose=Local To Component(LocalPose=Copy Pose From Mesh() {bUseAttachedParent=true})
    Alpha=@n10
    Translation="0,-3,0"
    Rotation="0,0,0"
    Scale="1,1,1"
   ) {
    translationMode=BMM_Additive
    translationSpace=BCS_BoneSpace
    alphaBoolBlend {blendOption=Linear}
    boneToModify {boneName=dyn_jacket_back}
    translation {y=-3}
   }
   n12 = "Transform (Modify) Bone - Bone: dyn_jacket_back_r"(ComponentPose=@n11, Alpha=@n10, Translation="0,-3,0", Rotation="0,0,0", Scale="1,1,1") {
    translationMode=BMM_Additive
    translationSpace=BCS_BoneSpace
    alphaBoolBlend {blendOption=Linear}
    boneToModify {boneName=dyn_jacket_back_r}
    translation {y=-3}
   }
   n13 = "Transform (Modify) Bone - Bone: dyn_jacket_back_r"(ComponentPose=@n12, Alpha=1, Translation="0,-1,0", Rotation="0,-3,0", Scale="1,1,1") {
    rotationMode=BMM_Additive
    rotationSpace=BCS_BoneSpace
    translationMode=BMM_Additive
    translationSpace=BCS_BoneSpace
    alphaBoolBlend {blendOption=Linear}
    boneToModify {boneName=dyn_jacket_back_r}
    rotation {yaw=-3}
    translation {y=-1}
   }
   n14 = "Transform (Modify) Bone - Bone: dyn_jacket_r"(ComponentPose=@n13, Alpha=1, Translation="0,2,0", Rotation="0,0,0", Scale="1,1,1") {
    rotationMode=BMM_Additive
    rotationSpace=BCS_BoneSpace
    translationMode=BMM_Additive
    translationSpace=BCS_BoneSpace
    alphaBoolBlend {blendOption=Linear}
    boneToModify {boneName=dyn_jacket_r}
    translation {y=2}
   }
   n15 = "Transform (Modify) Bone - Bone: dyn_jacket_l"(ComponentPose=@n14, Alpha=1, Translation="0,3,3", Rotation="0,0,0", Scale="1,1,1") {
    rotationMode=BMM_Additive
    rotationSpace=BCS_BoneSpace
    translationMode=BMM_Additive
    translationSpace=BCS_BoneSpace
    alphaBoolBlend {blendOption=Linear}
    boneToModify {boneName=dyn_jacket_l}
    translation {y=3, z=3}
   }
   n16 = "Transform (Modify) Bone - Bone: dyn_jacket_back"(ComponentPose=@n15, Alpha=Get Jacket_isSkydiving() -> (Jacket_isSkydiving:double), Translation="0,-6,0", Rotation="0,-20,0", Scale="1,1,1") {
    rotationMode=BMM_Additive
    rotationSpace=BCS_BoneSpace
    translationMode=BMM_Additive
    translationSpace=BCS_BoneSpace
    alphaBoolBlend {blendOption=Linear}
    boneToModify {boneName=dyn_jacket_back}
    rotation {yaw=-20}
    translation {y=-6}
   }
   n17 = Anim Dynamics - Bone: dyn_sleeve_l(ComponentPose=@n16, Alpha=1, ExternalForce="0,0,0") {
    angularDampingOverride=0.9999
    bGravityOverrideInSimSpace=true
    bOverrideAngularDamping=true
    bOverrideLinearDamping=true
    gravityScale=0.25
    linearDampingOverride=0.9999
    simulationSpace=RootRelative
    alphaBoolBlend {blendOption=Linear}
    boundBone {boneName=dyn_sleeve_l}
    componentAppliedLinearAccClamp=<900 900 900>
    componentLinearAccScale=<0.25 0.25 0.25>
    componentLinearVelScale=<0.25 0.25 0.25>
    physicsBodyDefinitions [{
      sphereCollisionRadius=0
      boundBone {boneName=dyn_sleeve_l}
      boxExtents=<10 3 5>
      constraintSetup {angularLimitsMax=<10 0 15>, angularLimitsMin=<-10 0 -15>, angularTarget=<0 0 0>}
      localJointOffset=<3 0 0>
     }]
    relativeSpaceBone {boneName=root}
   }
   n18 = Anim Dynamics - Bone: dyn_sleeve_r(ComponentPose=@n17, Alpha=1, ExternalForce="0,0,0") {
    angularDampingOverride=0.9999
    bGravityOverrideInSimSpace=true
    bOverrideAngularDamping=true
    bOverrideLinearDamping=true
    gravityScale=0.25
    linearDampingOverride=0.9999
    simulationSpace=RootRelative
    alphaBoolBlend {blendOption=Linear}
    boundBone {boneName=dyn_sleeve_r}
    componentAppliedLinearAccClamp=<900 900 900>
    componentLinearAccScale=<0.25 0.25 0.25>
    componentLinearVelScale=<0.25 0.25 0.25>
    physicsBodyDefinitions [{
      sphereCollisionRadius=0
      boundBone {boneName=dyn_sleeve_r}
      boxExtents=<10 3 5>
      constraintSetup {angularLimitsMax=<10 0 15>, angularLimitsMin=<-10 0 -15>, angularTarget=<0 0 0>}
      localJointOffset=<3 0 0>
     }]
    relativeSpaceBone {boneName=root}
   }
   n19 = RigidBody(ComponentPose=@n18) {
    bClampLinearTranslationLimitToRefPose=true
    bOverrideWorldGravity=true
    overridePhysicsAsset=&/F_MED_Demon_Hunter_Physics
    simulationSpace=BaseBoneSpace
    alphaBoolBlend {blendOption=Linear}
    baseBoneRef {boneName=root}
    componentAppliedLinearAccClamp=<1000 1000 1000>
    componentLinearAccScale=<1 1 1>
    componentLinearVelScale=<1 1 1>
    overrideWorldGravity {z=-980}
   }
   n20 = Get bIsTargeting()
   n21 = Anim Dynamics - Bone: dyn_hood(
    ComponentPose=@n19
    Alpha=Ease!(Function:EEasingFunc Enum=EaseOut, Alpha:double=SelectFloat(A=0.3, B=0.75, bPickA=@n20), A:float=1, B:float=0.5, BlendExp:double=1) -> (Result:float=~) {is=title K2Node_EaseFunction}
    ExternalForce=Get ext_force()
   ) {
    angularDampingOverride=0.9999
    bGravityOverrideInSimSpace=true
    bOverrideAngularDamping=true
    bOverrideLinearDamping=true
    linearDampingOverride=0.9999
    simulationSpace=RootRelative
    alphaBoolBlend {blendOption=Linear}
    boundBone {boneName=dyn_hood}
    componentAppliedLinearAccClamp=<4000 4000 4000>
    componentLinearAccScale=<0.25 0.25 0.25>
    componentLinearVelScale=<0.25 0.25 0.25>
    gravityOverride {z=-250}
    physicsBodyDefinitions [{
      sphereCollisionRadius=0
      boundBone {boneName=dyn_hood}
      boxExtents=<12 8 4>
      constraintSetup {angularLimitsMax=<0 15 30>, angularLimitsMin=<0 -15 -30>, angularTarget=<0 0 0>}
      localJointOffset=<4 0 0>
     }]
    relativeSpaceBone {boneName=spine_02}
   }
   Save cached pose 'CrouchingCache'(Pose=Component To Local(ComponentPose=@n21)) {cacheName=CrouchingCache}
  ]}
 EventGraph {nodes [
   Event BlueprintUpdateAnimation()
   > n1 = Sequence!(execute) -> (then_0, then_1, then_2:Exec)
   Delay(execute=[@n1.then_0, @n10.then], Duration=0.25)
   > Set ext_force(ext_force:Vector="0,0,1250") -> (Output_Get:Vector)
   > n2 = Delay(Duration=0.375)
   Do Once(execute=@n10.else, Reset=@n2.then)
   > Set ext_force(ext_force="0, 0, 0")
   n3 = BreakTransform(InTransform=GetDeltaTransformFromRefPose(self=GetOwnerSkeletalMeshComponent(), BoneName=head, BaseName=spine_01))
   Set Jacket_Rotate(execute=@n1.then_2, Jacket_Rotate:double=MapRangeClamped(Value=BreakRotator(InRot=@n3.Rotation).Yaw, InRangeA=0, InRangeB=-49)) -> (Output_Get:double)
   n4 = "float <= float"!(A:double=Get PawnSpeed2D(), B:double=421) -> (ReturnValue:bool) {is=title K2Node_CallFunction}
   n5 = Get bIsCrouching()
   > Set Jacket_Crouch(Jacket_Crouch:double=SelectFloat(A=0.75, B=0, bPickA=AND Boolean(A=@n4, B=@n5))) -> (Output_Get:double)
   n6 = Get bIsSkydiveDiveMode()
   n7 = Get bIsSkydiveFloating()
   n8 = Get bIsSkyDiving()
   n9 = Get bIsParachuteOpened()
   Set Jacket_isSkydiving(
    execute=@n1.then_1
    Jacket_isSkydiving:double=SelectFloat(A=1, B=0, bPickA=OR Boolean!(A=@n8, B=@n9, C:bool=@n7, D:bool=@n6) -> (ReturnValue))
   ) -> (Output_Get:double)
   n10 = Branch(Condition=Get bIsJumping())
   TryGetPawnOwner()
  ]}
]
variables [
 ext_force {type=Vector}
 Jacket_Crouch {type=number}
 Jacket_isSkydiving {type=number}
 Jacket_Rotate {type=number}
]
