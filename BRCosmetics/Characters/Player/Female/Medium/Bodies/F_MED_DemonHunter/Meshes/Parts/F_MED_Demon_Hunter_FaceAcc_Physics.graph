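# Physics asset for the face accessory part of the medium female Demon Hunter player cosmetic, covering its dynamic bangs and hair bones.
# A capsule body on the head serves as the anchor, with box bodies on the left and right bangs, left hair and back hair bones.
# Each hair bone is constrained directly to the head rather than chained, with identical narrow swing and very tight twist limits.
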PhysicsAsset /BRCosmetics/Characters/Player/Female/Medium/Bodies/F_MED_DemonHunter/Meshes/Parts/F_MED_Demon_Hunter_FaceAcc_Physics
bodies {
 dyn_bangs_l_01 [{shapeType=Box, extentX=1.51, extentY=2.01, extentZ=1.01, center=<0.3342 -0.0589 0>, rotation=<rot 0 -100.0001 0>}]
 dyn_bangs_l_02 [{shapeType=Box, extentX=2.51, extentY=3.01, extentZ=1.51, center=<0.8699 0 0>, rotation=<rot 0 -90 0>}]
 dyn_bangs_r_01 [{shapeType=Box, extentX=2.51, extentY=4.01, extentZ=1.51, center=<1.7491 0.0354 0.0002>, rotation=<rot 0 -85 0>}]
 dyn_bangs_r_02 [{shapeType=Box, extentX=1.51, extentY=6.51, extentZ=3.01, center=<1.6425 0.3909 -0.0797>, rotation=<rot 55.0756 -89.9351 -0.8148>}]
 dyn_hair_back [{shapeType=Box, extentX=12.51, extentY=4.01, extentZ=4.01, center=<-0.0474 0 0.7387>, rotation=<rot 0 -89.9999 -0.0014>}]
 dyn_hair_l [{shapeType=Box, extentX=3.01, extentY=3.01, extentZ=2.01, center=<0.7096 0 0>, rotation=<rot 0 -89.9999 19.9995>}]
 head [{shapeType=Capsule, length=6.0946, radius=6.824, center=<0.9446 3.2649 -0.1028>, rotation=<rot 0.9787 -89.9998 -90.5167>}]
}
constraints [
 {bone1Name=dyn_bangs_r_01, bone2Name=head, swing1LimitDegrees=20, swing2LimitDegrees=20, twistLimitDegrees=3}
 {bone1Name=dyn_bangs_r_02, bone2Name=head, swing1LimitDegrees=20, swing2LimitDegrees=20, twistLimitDegrees=3}
 {bone1Name=dyn_bangs_l_01, bone2Name=head, swing1LimitDegrees=20, swing2LimitDegrees=20, twistLimitDegrees=3}
 {bone1Name=dyn_bangs_l_02, bone2Name=head, swing1LimitDegrees=20, swing2LimitDegrees=20, twistLimitDegrees=3}
 {bone1Name=dyn_hair_l, bone2Name=head, swing1LimitDegrees=20, swing2LimitDegrees=20, twistLimitDegrees=3}
 {bone1Name=dyn_hair_back, bone2Name=head, swing1LimitDegrees=20, swing2LimitDegrees=20, twistLimitDegrees=3}
]
